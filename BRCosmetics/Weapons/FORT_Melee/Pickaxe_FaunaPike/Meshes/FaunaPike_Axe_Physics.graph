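# Physics asset for the FaunaPike pickaxe cosmetic, defining a dangling chain on the two-handed melee weapon.
# Eight capsule bodies are constrained in series from MSTR_chain, with tight twist limits, a mostly one-axis swing at the root and wider swing along the links.
# The last two links have larger capsules, forming a bulkier chain end.
# Unconstrained capsules cover the handle and the left and right thigh and calf bones, apparently as collision shapes for the chain.
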PhysicsAsset /BRCosmetics/Weapons/FORT_Melee/Pickaxe_FaunaPike/Meshes/FaunaPike_Axe_Physics
bodies {
 MSTR_chain [{shapeType=Capsule, length=0.505, radius=0.505, shapeName=MSTR_chain_capsule, rotation=<rot 0 -90 0>}]
 Melee_TwoHanded_Handle [{shapeType=Capsule, length=53.7654, radius=13.2109, center=<0 0 26.403>}]
 calf_l [{shapeType=Capsule, length=50, radius=8, center=<-30 -2 0>, rotation=<rot 90 0 0>}]
 calf_r [{shapeType=Capsule, length=50, radius=8, center=<30 2 0>, rotation=<rot 90 0 0>}]
 dyn_chain_01 [{shapeType=Capsule, length=2.3503, radius=1.42, shapeName=dyn_chain_01_capsule, center=<2 0 0>, rotation=<rot -90 0 0>}]
 dyn_chain_02 [{shapeType=Capsule, length=2.3503, radius=1.42, shapeName=dyn_chain_02_capsule, center=<2.5 0 0>, rotation=<rot -90 0 0>}]
 dyn_chain_03 [{shapeType=Capsule, length=2.3498, radius=1.42, shapeName=dyn_chain_03_capsule, center=<2.3 0 0>, rotation=<rot -90 0 0>}]
 dyn_chain_04 [{shapeType=Capsule, length=2.3523, radius=1.4199, shapeName=dyn_chain_04_capsule, center=<2 0 0>, rotation=<rot -90 0 0>}]
 dyn_chain_05 [{shapeType=Capsule, length=2.3558, radius=1.4201, shapeName=dyn_chain_05_capsule, center=<1.85 0 0>, rotation=<rot -90 0 0>}]
 dyn_chain_06 [{shapeType=Capsule, length=2.3554, radius=1.4201, shapeName=dyn_chain_06_capsule, center=<1.6 0 0>, rotation=<rot -90 0 0>}]
 dyn_chain_07 [{shapeType=Capsule, length=3.3968, radius=2.2925, shapeName=dyn_chain_07_capsule, center=<2.4 0 0>, rotation=<rot -90 0 0>}]
 dyn_chain_08 [{shapeType=Capsule, length=4.9667, radius=5.489, shapeName=dyn_chain_08_capsule, center=<4 -1 0>, rotation=<rot -90 15 0>}]
 thigh_l [{shapeType=Capsule, length=40, radius=8, center=<-25 0 0>, rotation=<rot 90 0 0>}]
 thigh_r [{shapeType=Capsule, length=40, radius=8, center=<25 0 0>, rotation=<rot 90 0 0>}]
}
constraints [
 {bone1Name=dyn_chain_01, bone2Name=MSTR_chain, swing1LimitDegrees=80, swing2LimitDegrees=5, twistLimitDegrees=5}
 {bone1Name=dyn_chain_02, bone2Name=dyn_chain_01, swing1LimitDegrees=45, swing2LimitDegrees=45, twistLimitDegrees=5}
 {bone1Name=dyn_chain_03, bone2Name=dyn_chain_02, swing1LimitDegrees=45, swing2LimitDegrees=45, twistLimitDegrees=5}
 {bone1Name=dyn_chain_04, bone2Name=dyn_chain_03, swing1LimitDegrees=45, swing2LimitDegrees=45, twistLimitDegrees=5}
 {bone1Name=dyn_chain_05, bone2Name=dyn_chain_04, swing1LimitDegrees=45, swing2LimitDegrees=45, twistLimitDegrees=5}
 {bone1Name=dyn_chain_06, bone2Name=dyn_chain_05, swing1LimitDegrees=45, swing2LimitDegrees=45, twistLimitDegrees=5}
 {bone1Name=dyn_chain_07, bone2Name=dyn_chain_06, swing1LimitDegrees=45, swing2LimitDegrees=45, twistLimitDegrees=5}
 {bone1Name=dyn_chain_08, bone2Name=dyn_chain_07, swing1LimitDegrees=45, swing2LimitDegrees=60, twistLimitDegrees=5}
]
